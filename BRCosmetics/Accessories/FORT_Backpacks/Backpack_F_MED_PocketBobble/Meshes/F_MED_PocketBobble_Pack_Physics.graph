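# Physics asset for the PocketBobble backpack mesh, providing capsule bodies and joints for the pack and its charm.
# A large capsule covers main_PocketBobble, with smaller capsules on base_charm and dyn_charm.
# base_charm is angularly locked to the main body, forming an anchor for dyn_charm, whose joint carries its own swing and twist limit settings.
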PhysicsAsset /BRCosmetics/Accessories/FORT_Backpacks/Backpack_F_MED_PocketBobble/Meshes/F_MED_PocketBobble_Pack_Physics
bodies {
 base_charm [{shapeType=Capsule, length=0.505, radius=0.505, shapeName=base_charm_capsule, rotation=<rot 0 -90 0>}]
 dyn_charm [{shapeType=Capsule, length=2, radius=2, shapeName=dyn_charm_capsule, center=<6 0 0>, rotation=<rot -90 0 0>}]
 main_PocketBobble [{shapeType=Capsule, length=56.1023, radius=7.414, center=<-1.9319 -4.2961 -4.3092>, rotation=<rot 5.0049 -95.7546 40.0442>}]
}
constraints [
 {
  bone1Name=base_charm
  bone2Name=main_PocketBobble
  swing1LimitDegrees=45
  swing1Motion=Locked
  swing2LimitDegrees=45
  swing2Motion=Locked
  twistLimitDegrees=45
  twistMotion=Locked
 }
 {bone1Name=dyn_charm, bone2Name=base_charm, swing1LimitDegrees=40, swing2LimitDegrees=40, twistLimitDegrees=0}
]
