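# Animation Blueprint for the PocketBobble backpack cosmetic, adding procedural motion to the pack, bows and charm.
# Spine_05's yaw relative to spine_01 drives an additive roll and shift of the pack bone.
# AnimDynamics springs the bows, a RigidBody node follows, and charm dynamics blend in only in the front end.
# An open parachute offsets the main bone, fixed pitches adjust the charm base, and computed jump and skydive values go unread.
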AnimBlueprint /BRCosmetics/Accessories/FORT_Backpacks/Backpack_F_MED_PocketBobble/Meshes/F_MED_PocketBobble_Pack_AnimBP
target_skeleton=&/F_MED_PocketBobble_Pack_Skeleton
graphs [
 AnimGraph {nodes [
   n1 = "Transform (Modify) Bone - Bone: pack"(
    ComponentPose=Local To Component(LocalPose=Input Pose())
    Translation=Get backpack_translation()
    Rotation=Get backpack_rotator() -> (backpack_rotator:Rotator)
   ) {
    rotationMode=BMM_Additive
    rotationSpace=BCS_BoneSpace
    translationMode=BMM_Additive
    translationSpace=BCS_BoneSpace
    boneToModify {boneName=pack}
   }
   n2 = Anim Dynamics - Bone: dyn_bow_b(ComponentPose=@n1, Alpha=1, ExternalForce="0,0,0") {
    angularDampingOverride=1
    angularSpringConstant=20000
    bAngularSpring=true
    bOverrideAngularDamping=true
    bUseGravityOverride=true
    gravityScale=0.35
    simulationSpace=RootRelative
    boundBone {boneName=dyn_bow_b}
    componentAppliedLinearAccClamp=<3000 3000 3000>
    componentLinearAccScale=<0.4 0.4 0.4>
    componentLinearVelScale=<0.4 0.4 0.4>
    gravityOverride {z=-980}
    physicsBodyDefinitions [{
      boundBone {boneName=dyn_bow_b}
      boxExtents=<20 20 20>
      constraintSetup {angularLimitsMax=<0 10 12>, angularLimitsMin=<0 -40 -12>, angularTarget=<1 0 0>}
      localJointOffset=<4 0 0>
     }]
   }
   n3 = Anim Dynamics - Bone: dyn_bow_a(ComponentPose=@n2, Alpha=1, ExternalForce="0,0,0") {
    angularDampingOverride=1
    angularSpringConstant=20000
    bAngularSpring=true
    bOverrideAngularDamping=true
    bUseGravityOverride=true
    gravityScale=0.35
    simulationSpace=RootRelative
    boundBone {boneName=dyn_bow_a}
    componentAppliedLinearAccClamp=<3000 3000 3000>
    componentLinearAccScale=<0.4 0.4 0.4>
    componentLinearVelScale=<0.4 0.4 0.4>
    gravityOverride {z=-980}
    physicsBodyDefinitions [{
      boundBone {boneName=dyn_bow_a}
      boxExtents=<20 20 20>
      constraintSetup {angularLimitsMax=<0 40 13>, angularLimitsMin=<0 -20 -13>, angularTarget=<1 0 0>}
      localJointOffset=<3 0 0>
     }]
    simSpaceSettings {simSpaceAngularAlpha=1}
   }
   n4 = RigidBody(ComponentPose=@n3) {
    bOverrideWorldGravity=true
    overridePhysicsAsset=&/F_MED_PocketBobble_Pack_Physics
    simulationSpace=BaseBoneSpace
    baseBoneRef {boneName=root}
    componentAppliedLinearAccClamp=<3000 3000 3000>
    overrideWorldGravity {z=-980}
    simSpaceSettings {
     maxAngularAcceleration=40
     maxAngularVelocity=80
     maxLinearAcceleration=1500
     maxLinearVelocity=700
     worldAlpha=1
     externalLinearDragV=<1 1 1>
    }
   }
   n5 = Anim Dynamics - Bone: dyn_charm(ComponentPose=@n4, Alpha=Get FrontEnd() -> (FrontEnd:double), ExternalForce="0,0,0") {
    angularDampingOverride=1
    angularSpringConstant=20000
    bAngularSpring=true
    bOverrideAngularDamping=true
    bUseGravityOverride=true
    gravityScale=0.35
    simulationSpace=BoneRelative
    boundBone {boneName=dyn_charm}
    componentAppliedLinearAccClamp=<3000 3000 3000>
    componentLinearAccScale=<0.4 0.4 0.4>
    componentLinearVelScale=<0.4 0.4 0.4>
    gravityOverride {z=-980}
    physicsBodyDefinitions [{
      boundBone {boneName=dyn_charm}
      boxExtents=<20 20 20>
      constraintSetup {angularLimitsMax=<0 12 8>, angularLimitsMin=<0 -12 -12>, angularTarget=<1 0 0>}
      localJointOffset=<2 0 0>
     }]
   }
   n6 = "Transform (Modify) Bone - Bone: main_PocketBobble"!(ComponentPose=@n5, bAlphaBoolEnabled:bool=Get bIsParachuteOpened(), Translation="0,-3,0", Rotation="0,0,-10", Scale="1,1,1") -> (Pose) {
    alphaInputType=Bool
    rotationMode=BMM_Additive
    rotationSpace=BCS_BoneSpace
    translationMode=BMM_Additive
    translationSpace=BCS_BoneSpace
    boneToModify {boneName=main_PocketBobble}
    rotation {roll=-10}
    translation {y=-3}
   }
   n7 = "Transform (Modify) Bone - Bone: base_charm"(ComponentPose=@n6, Alpha=1, Translation="0,0,0", Rotation="25,0,0", Scale="1,1,1") {
    rotationMode=BMM_Additive
    rotationSpace=BCS_BoneSpace
    translationMode=BMM_Additive
    translationSpace=BCS_BoneSpace
    boneToModify {boneName=base_charm}
    rotation {pitch=25}
   }
   n8 = "Transform (Modify) Bone - Bone: base_charm"(ComponentPose=@n7, Alpha=1, Translation="0,0,0", Rotation="50,0,0", Scale="1,1,1") {
    rotationMode=BMM_Additive
    rotationSpace=BCS_BoneSpace
    translationMode=BMM_Additive
    translationSpace=BCS_BoneSpace
    alphaScaleBiasClamp {bInterpResult=true, interpSpeedDecreasing=9, interpSpeedIncreasing=9}
    boneToModify {boneName=base_charm}
    rotation {pitch=50}
   }
   Output Pose(Result=Component To Local(ComponentPose=@n8)) {layerGroup=None}
  ]}
 EventGraph {nodes [
   Event BlueprintUpdateAnimation()
   > n1 = Sequence()
   n2 = GetOwnerSkeletalMeshComponent()
   n3 = BreakRotator(InRot=BreakTransform(InTransform=GetDeltaTransformFromRefPose(self=@n2, BoneName=spine_05, BaseName=spine_01)).Rotation)
   > Set backpack_rotator(backpack_rotator:Rotator=MakeRotator(Roll=MapRangeClamped(Value=@n3.Yaw, InRangeA=0, InRangeB=-30, OutRangeB=35))) -> (Output_Get:Rotator)
   > Set backpack_translation(backpack_translation=MakeVector(Y=MapRangeClamped(Value=@n3.Yaw, InRangeA=0, InRangeB=-50, OutRangeB=-7)))
   n4 = Get bIsJumping()
   n5 = Get bIsFalling()
   Set jump(execute=@n1.then_1, jump:double="To Float (Boolean)"(InBool=OR Boolean(A=@n4, B=@n5))) -> (Output_Get:double)
   > Set Skydive_floating(Skydive_floating:double="To Float (Boolean)"(InBool=Get bIsSkydiveFloating())) -> (Output_Get:double)
   > Set FrontEnd(FrontEnd:double="To Float (Boolean)"(InBool=Get bIsInFrontEnd())) -> (Output_Get:double)
  ]}
]
variables [
 backpack_rotator {type=Rotator}
 backpack_translation {type=Vector}
 FrontEnd {type=number}
 Gravity_Override {type=Vector}
 jump {type=number}
 parachute_olnyopen {type=number}
 rotate left {type=number}
 rotate_up {type=number}
 Skydive_floating {type=number}
]
